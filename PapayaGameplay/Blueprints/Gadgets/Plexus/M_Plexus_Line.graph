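# Unlit, additive, two-sided material for Niagara mesh particles that draws the lines of the Plexus gadget effect.
# Particle color times a dynamic Emissive parameter feeds the FX emissive adjustment function.
# An offset chain expands UV.y along the cross product of view direction and transformed X axis, scaling width with pixel depth over screen height for camera-facing lines.
# The graph also holds a Plexus texture mask and a PlexusAlpha collection parameter.
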Material /PapayaGameplay/Blueprints/Gadgets/Plexus/M_Plexus_Line
graph {
 properties {
  bUsedWithNiagaraMeshParticles=true
  blendMode=BLEND_Additive
  pixelDepthOffsetMode=PDOM_Legacy
  shadingModel=MSM_Unlit
  twoSided=true
  usedShadingModels=Unlit
 }
 nodes [
  CollectionParameter() {collection=/Game/Environments/Papaya/Materials/MPC_PapayaGadgetControls, parameterName=PlexusAlpha}
  MaterialFunctionCall(Emissive=Multiply(
    A=ParticleColor().RGB
    B=DynamicParameter() {parameterIndex=1, defaultValue=<rgba 1 1 1 5>, paramNames [Emissive,Param2,Param3,Wide]}.Emissive
   )) {materialFunction=&MF_FX_EmissiveAdjustments}
  n1 = ScalarParameter() {defaultValue=1, parameterName=ssssssss}
  n2 = ViewSize()
  n3 = Divide(A=PixelDepth().R, B=Multiply(A=ComponentMask(None=@n2) {g=true}, B=ScalarParameter() {defaultValue=100, parameterName=ScreenWidth}))
  n4 = Subtract(A=TextureCoordinate()) {constB=0.5}
  n5 = Subtract(A=WorldPosition().XYZ, B=CameraPositionWS())
  n6 = Constant3Vector() {constant=<rgba 1 0 0 1>}
  Multiply(A=Multiply(A=Add(A=@n1, B=@n3), B=ComponentMask(None=@n4) {g=true}), B=CrossProduct(A=Normalize(VectorInput=@n5), B=Transform(None=@n6)))
  Multiply(A=TextureSample() {texture=&/T_Plexus}.R, B=Constant() {r=1})
 ]
}
