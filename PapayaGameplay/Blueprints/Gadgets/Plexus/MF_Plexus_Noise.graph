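# Material function for the Plexus gadget that generates animated noise outputs for world position offset and emissive.
# World position, shifted by time and divided by a scale parameter, samples a vector noise node.
# The noise times an intensity parameter becomes NoiseWPO.
# NoiseEmissive sums red twice, blue and a connection factor (green is unused), then saturates and computes per vertex through a VertexInterpolator.
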MaterialFunction /PapayaGameplay/Blueprints/Gadgets/Plexus/MF_Plexus_Noise
graph {nodes [
  n1 = WorldPosition()
  n2 = Multiply(A=Time(), B=ScalarParameter() {defaultValue=20, parameterName=F_NoiseSpeed})
  n3 = VectorNoise(World Position=Divide(A=Add(A=@n1.XYZ, B=@n2), B=ScalarParameter() {defaultValue=2048, parameterName=F_NoiseWPscale})) {noiseFunction=VNF_VectorALU}
  FunctionOutput(None=Multiply(A=@n3, B=ScalarParameter() {defaultValue=50000, parameterName=Mesh_NoiseIntensity})) {outputName=NoiseWPO}
  n4 = MaterialFunctionCall(Float3=@n3) {materialFunction=&BreakOutFloat3Components}
  n5 = Add(A=Add(A=Add(A=@n4.R, B=@n4.R), B=@n4.B), B=ScalarParameter() {defaultValue=0.25, parameterName=Mesh_ConnectionFactor})
  FunctionOutput(None=VertexInterpolator(VS=Saturate(None=@n5)).PS) {outputName=NoiseEmissive}
 ]}
