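# Unlit additive material for the spline meshes of the Papaya Plexus gadget effect.
# It builds a camera-facing ribbon offset by crossing a transformed X axis with the view direction, scaled by centered V coordinate and a width parameter, then adds MF_Plexus_Noise displacement.
# The shared gadget color function is included, and the Plexus texture is modulated by noise emissive and the PlexusAlpha collection parameter.
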Material /PapayaGameplay/Blueprints/Gadgets/Plexus/M_Plexus_Mesh_Spline
graph {
 properties {blendMode=BLEND_Additive, pixelDepthOffsetMode=PDOM_Legacy, shadingModel=MSM_Unlit, usedShadingModels=Unlit}
 nodes [
  n1 = Subtract(A=ComponentMask(None=TextureCoordinate()) {g=true}) {constB=0.5}
  n2 = ScalarParameter() {defaultValue=256, parameterName=Wide}
  n3 = Transform(None=Constant3Vector() {constant=<rgba 1 0 0 1>})
  n4 = Normalize(VectorInput=Subtract(A=WorldPosition().XYZ, B=CameraPositionWS()))
  n5 = MaterialFunctionCall() {materialFunction=&/MF_Plexus_Noise}
  Add(A=Multiply(A=Multiply(A=@n1, B=@n2), B=CrossProduct(A=@n3, B=@n4)), B=@n5.NoiseWPO)
  MaterialFunctionCall!(
   Media Player TexCoord
   Media Color Code Coordinate
   Media Texture Resolution
   Base Color Parameter
   Base Color Mix
   Media Color Code Mix
   Intermission Color Parameter
   Intermission Color Mix
   Desaturation
  ) {materialFunction=/Game/Environments/Papaya/Materials/MaterialFunctions/MF_Papaya_GadgetColor}
  Multiply(
   A=CollectionParameter() {collection=/Game/Environments/Papaya/Materials/MPC_PapayaGadgetControls, parameterName=PlexusAlpha}
   B=Multiply(A=TextureSample() {texture=&/T_Plexus}.R, B=@n5.NoiseEmissive)
  )
 ]
}
