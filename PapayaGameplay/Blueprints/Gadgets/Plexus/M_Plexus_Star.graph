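# Additive, unlit, two-sided material for the stars of the Plexus gadget, usable on Niagara sprites.
# The green channel of T_Plexus is tinted and sharpened by a power, then multiplied by MF_Papaya_GadgetColor for emissive, while the PlexusAlpha collection parameter scales opacity.
# The UsingForMesh switch expands mesh quads into camera-facing billboards from UVs via World Position Offset, adding MF_Plexus_Noise displacement and brightness modulation.
# SphereMask and sine branches are unconnected.
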Material /PapayaGameplay/Blueprints/Gadgets/Plexus/M_Plexus_Star
graph {
 properties {
  bUseMaterialAttributes=true
  bUsedWithNiagaraSprites=true
  blendMode=BLEND_Additive
  pixelDepthOffsetMode=PDOM_Legacy
  shadingModel=MSM_Unlit
  twoSided=true
  usedShadingModels=Unlit
 }
 nodes [
  n1 = TextureCoordinate() {uTiling=2, vTiling=2}
  OneMinus(None=Sine(None=@n1))
  n2 = Subtract(A=@n1)
  SphereMask(
   A=Power(Base=Abs(None=@n2), Exp=ScalarParameter() {defaultValue=0.2, parameterName=Param_2})
   B=Constant()
   Radius=ScalarParameter() {defaultValue=1.5}
   Hardness=ScalarParameter() {parameterName=Param_1}
  )
  n3 = Multiply(
   A=TextureSample() {texture=&/T_Plexus}.G
   B=VectorParameter() {parameterName=C, defaultValue=<rgba 2 2 2 0>}.RGB
  )
  n4 = Power(Base=@n3, Exp=ScalarParameter() {defaultValue=5.5, parameterName=CExp})
  n5 = MaterialFunctionCall!(
   Media Player TexCoord
   Media Color Code Coordinate
   Media Texture Resolution
   Base Color Parameter
   Base Color Mix
   Media Color Code Mix
   Intermission Color Parameter
   Intermission Color Mix
   Desaturation
  ) {materialFunction=/Game/Environments/Papaya/Materials/MaterialFunctions/MF_Papaya_GadgetColor}
  n6 = SetMaterialAttributes!(
   MaterialAttributes
   Emissive Color=Multiply(A=@n4, B=@n5.Result)
   Opacity=Multiply(A=CollectionParameter() {collection=/Game/Environments/Papaya/Materials/MPC_PapayaGadgetControls, parameterName=PlexusAlpha}, B=@n3)
  ) {attributeSetTypes [#1, #2]}
  n7 = Transform(None=Constant3Vector() {constant=<rgba 1 0 0 0>}) {transformSourceType=TRANSFORMSOURCE_View}
  n8 = MaterialFunctionCall(Float2=Subtract(A=TextureCoordinate()) {constB=0.5}) {materialFunction=&BreakOutFloat2Components}
  n9 = Transform(None=Constant3Vector() {constant=<rgba 0 1 0 0>}) {transformSourceType=TRANSFORMSOURCE_View}
  n10 = Multiply(A=Add(A=Multiply(A=@n7, B=@n8.R), B=Multiply(A=@n9, B=@n8.G)), B=ScalarParameter() {defaultValue=50, parameterName=Scale})
  n11 = MaterialFunctionCall() {materialFunction=&/MF_Plexus_Noise}
  n12 = Multiply(
   A=MaterialFunctionCall(X=@n11.NoiseEmissive, Y=@n11.NoiseEmissive, Z=@n11.NoiseEmissive) {materialFunction=&MakeFloat3}.Result
   B=@n4
  )
  StaticSwitchParameter(
   True=SetMaterialAttributes!(MaterialAttributes=@n6, World Position Offset=Add(A=@n10, B=@n11.NoiseWPO), Emissive Color=Multiply(A=@n12, B=@n5.Result)) {attributeSetTypes [#3, #1]}
   False=@n6
  ) {parameterName=UsingForMesh}
 ]
}
