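# Niagara system for the Plexus gadget: a long-lived, orbiting network of particles drawn with mesh and sprite renderers.
# A single emitter bursts 500 particles, placed by a custom stacked, bow-curve initializer and orbited at User.MovingSpeed.
# A custom update reads the same emitter's particles, apparently to link neighbors within User.LinkRange.
# Color samples User.BaseColorCurve using the fractional Base Color Parameter from the Papaya_Jukebox_NPC parameter collection.
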NiagaraSystem /PapayaGameplay/Blueprints/Gadgets/Plexus/NS_Plexus_System
dependencies {
 usedDataInterfaces [&NiagaraDataInterfaceCurve, &NiagaraDataInterfaceParticleRead]
 usedDynamicInputs []
 usedModules [&SystemState]
}
emitters {NE_Jerky_Plexus {
  emitterUpdateScript [
   &EmitterState(
    Inactive Response=::NewEnumerator1
    Life Cycle Mode=::NewEnumerator1
    Loop Duration=999
    Scalability Mode=::NewEnumerator1
    UseLoopDelay=false
   )
   &SpawnBurst_Instantaneous(Spawn Count=500, Use Spawn Probability=false)
   &/C_Plexus_OrbitUpdate(OrbitDeltaAngle:NiagaraFloat=$User.MovingSpeed) {module=C_Jerky_OrbitUpdate}
  ]
  particleSpawnScript [
   /Niagara/Modules/Spawn/Initialization/InitializeParticle(
    Lifetime=999
    Mesh Scale=<1 0.25 0.25>
    Module.Write Color=true
    Module.Write Lifetime=true
    Module.Write Mass=true
    Module.Write Material Random=true
    Module.Write Position=false
    Module.Write Scale=true
    Module.Write SpriteRotation=true
    Module.Write SpriteSize=true
    Position=&SimulationPosition()
    Sprite Rotation=&UniformRangedFloat(Minimum=&UniformAOrBFloat(A=0, B=0), Maximum=1)
    Sprite Size=&Vector2DFromFloat(Value=&FloatFromCurve(
      FloatCurve=DataInterface() {propertyValues {Curve {keys [{arriveTangent=-217.3979, interpMode=RCIM_Cubic, leaveTangent=-217.3978, tangentMode=RCTM_User, time=0, value=150}, <at 1 40 RCIM_Cubic>]}}}
      CurveIndex=&UniformRangedFloat(Minimum=0, Maximum=1)
      Scale Curve=0.25
     ))
   ) {module=InitializeParticle001}
   &/C_Plexus_Initial(
    BowShape=DataInterface() {propertyValues {Curve {keys [
        {arriveTangent=1.0868, interpMode=RCIM_Cubic, leaveTangent=1.0868, tangentMode=RCTM_User, time=0, value=-0.0383}
        {arriveTangent=0.0383, interpMode=RCIM_Cubic, leaveTangent=0.0383, time=0.5, value=1}
        <at 1 0 RCIM_Cubic>
       ]}}}
    SinPeriod=180
    Stack=5
    VolumeSize:Vector3f=<750 750 750>
   ) {module=C_Jerky_Plexus_Initial}
  ]
  particleUpdateScript [
   &ParticleState()
   &/C_Plexus_Update(
    ConnectMaxRange:NiagaraFloat=$User.LinkRange
    EmissiveFactorCurve=DataInterface() {propertyValues {Curve {keys [<at 0.5 1>]}}}
    NewVariable=DataInterface() {propertyValues {EmitterBinding {bindingMode=Other, emitterName=NE_Jerky_Plexus}}}
    Stack=5
   ) {module=C_Jerky_Plexus_Update}
   &CameraOffset(Camera Offset Amount=-100)
   &ScaleSpriteSize(Scale Factor=&Vector2DFromFloat(Value=&UniformRangedFloat(Minimum=1, Maximum=1)))
   &ScaleColor(
    Scale RGB=&MakeVectorFromLinearColorRGB(Color=&ColorFromCurve(ColorCurve=$User.BaseColorCurve, CurveIndex=&Frac_Float(Float=$NPC.Papaya_Jukebox_NPC.Base Color Parameter)))
    ScaleA=true
    ScaleRGB=true
   )
  ]
  rendererClasses [&NiagaraMeshRendererProperties, &NiagaraSpriteRendererProperties]
  renderers [
   {rendererClass=&NiagaraMeshRendererProperties, rendererIndex=0}
   {rendererClass=&NiagaraSpriteRendererProperties, rendererIndex=1}
  ]
 }}
summary {
 systemName=NS_Plexus_System
 emitters [{
   emitterName=NE_Jerky_Plexus
   rendererClasses [&NiagaraMeshRendererProperties, &NiagaraSpriteRendererProperties]
  }]
 userVariables [
  User.BaseColorCurve {
   type=NiagaraDataInterfaceColorCurve
   description=""
   defaultValue {
    struct=NiagaraExt_VariableValue_DataInterface
    value {
     dataInterface=&self:NiagaraDataInterfaceColorCurve_0
     dataInterfaceClass=&NiagaraDataInterfaceColorCurve
    }
   }
  }
  User.LinkRange {type=NiagaraFloat, defaultValue=17500, description=""}
  User.MovingSpeed {type=NiagaraFloat, defaultValue=0.25, description=""}
 ]
}
